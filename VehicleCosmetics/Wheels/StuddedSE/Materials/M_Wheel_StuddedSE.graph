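# Masked material for the StuddedSE cosmetic vehicle wheel, covering rim and tire in one skeletal-mesh shader.
# Packed masks split primary, accent, panning gradient and emissive regions, with rim and tire results lerped for color, roughness, metallic and normal.
# A Fresnel glow pulses via a time-driven sine, while vertex-color blue limits where the rim alpha mask cuts out.
# Attributes then pass through the shared vehicle damage and de-rez transition functions.
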Material /VehicleCosmetics/Wheels/StuddedSE/Materials/M_Wheel_StuddedSE
graph {
 properties {
  bAutomaticallySetUsageInEditor=false
  bUseMaterialAttributes=true
  bUsedWithMeshDeformer=true
  bUsedWithSkeletalMesh=true
  blendMode=BLEND_Masked
  pixelDepthOffsetMode=PDOM_Legacy
  usedShadingModels=Default Lit
 }
 nodes [
  n1 = NamedRerouteUsage()
  n2 = NamedRerouteUsage()
  n3 = NamedRerouteUsage()
  n4 = NamedRerouteUsage()
  n5 = ScalarParameter() {defaultValue=0.7, group=0_EmissiveColor, parameterName=EmissiveRoughness, sortPriority=5}
  n6 = NamedRerouteUsage()
  n7 = ScalarParameter() {defaultValue=0.25, group=2_GradientColor, parameterName=GradientRoughness, sortPriority=2}
  n8 = NamedRerouteUsage()
  n9 = ScalarParameter() {defaultValue=0.55, group=1_BaseColor, parameterName=AccentColor_Roughness, sortPriority=1}
  n10 = NamedRerouteUsage()
  n11 = Constant() {r=0.5}
  n12 = Add(A=Add(A=Multiply(A=@n4, B=@n5), B=Multiply(A=@n6, B=@n7)), B=Add(A=Multiply(A=@n8, B=@n9), B=Multiply(A=@n10, B=@n11)))
  n13 = NamedRerouteUsage()
  n14 = ScalarParameter() {defaultValue=0.25, group=2_GradientColor, parameterName=GradientRoughness, sortPriority=2}
  n15 = NamedRerouteUsage()
  n16 = Constant() {r=0.5}
  n17 = Add(
   A=Add(A=Multiply(A=@n13, B=@n14), B=Multiply(A=@n15, B=@n16))
   B=Multiply(A=NamedRerouteUsage(), B=ScalarParameter() {defaultValue=0.55, group=1_BaseColor, parameterName=AccentColor_Roughness, sortPriority=1})
  )
  n18 = NamedRerouteUsage()
  n19 = NamedRerouteUsage()
  n20 = ScalarParameter() {defaultValue=0.6, group=0_EmissiveColor, parameterName=EmissiveMetallic, sortPriority=6}
  n21 = NamedRerouteUsage()
  n22 = ScalarParameter() {group=2_GradientColor, parameterName=GradientMetallic, sortPriority=3}
  n23 = NamedRerouteUsage()
  n24 = ScalarParameter() {defaultValue=0.5, group=1_BaseColor, parameterName=AccentColor_Metallic, sortPriority=2}
  n25 = NamedRerouteUsage()
  n26 = Constant() {r=0.5}
  n27 = Add(A=Add(A=Multiply(A=@n19, B=@n20), B=Multiply(A=@n21, B=@n22)), B=Add(A=Multiply(A=@n23, B=@n24), B=Multiply(A=@n25, B=@n26)))
  n28 = NamedRerouteUsage()
  n29 = ScalarParameter() {group=2_GradientColor, parameterName=GradientMetallic, sortPriority=3}
  n30 = NamedRerouteUsage()
  n31 = ScalarParameter() {defaultValue=0.5, group=1_BaseColor, parameterName=AccentColor_Metallic, sortPriority=2}
  n32 = Add(A=Add(A=Multiply(A=@n28, B=@n29), B=Multiply(A=@n30, B=@n31)), B=Multiply(A=NamedRerouteUsage(), B=Constant() {r=0.5}))
  n33 = NamedRerouteUsage()
  n34 = Multiply(A=NamedRerouteUsage(), B=NamedRerouteUsage())
  n35 = Add(A=Multiply(A=NamedRerouteUsage(), B=NamedRerouteUsage()), B=Multiply(A=NamedRerouteUsage(), B=NamedRerouteUsage()))
  n36 = Multiply(A=NamedRerouteUsage(), B=NamedRerouteUsage())
  n37 = Add(A=Multiply(A=NamedRerouteUsage(), B=NamedRerouteUsage()), B=Multiply(A=NamedRerouteUsage(), B=NamedRerouteUsage()))
  n38 = SetMaterialAttributes!(
   MaterialAttributes
   Emissive Color=Multiply(A=Multiply(A=@n1, B=@n2), B=OneMinus(None=@n3))
   Normal=LinearInterpolate(A=NamedRerouteUsage(), B=NamedRerouteUsage(), Alpha=NamedRerouteUsage())
   Roughness=Saturate(None=LinearInterpolate(A=@n12, B=@n17, Alpha=@n18))
   Metallic=Saturate(None=LinearInterpolate(A=@n27, B=@n32, Alpha=@n33))
   Base Color=LinearInterpolate(A=Add(A=@n34, B=@n35), B=Add(A=@n36, B=@n37), Alpha=NamedRerouteUsage())
   Opacity Mask=LinearInterpolate(
    B=TextureSample() {samplerType=SAMPLERTYPE_LinearColor, texture=/VehicleCosmetics/Wheels/StuddedSE/Textures/T_Rim_StuddedSE_AlphaMask}.R
    Alpha=VertexColor().B
   ) {constA=1}
  ) {attributeSetTypes [
    #1
    #2
    #3
    #4
    #5
    #6
   ]}
  MaterialFunctionCall(In=MaterialFunctionCall(Material=@n38) {materialFunction=&MF_BaseVehicle_Damage}.Result) {materialFunction=&MF_VehicleDeRez_Transition}
  n39 = Fresnel(Normal=Transform(None=NamedRerouteUsage())) {exponent=1}
  n40 = ComponentMask(None=OneMinus(None=CameraVectorWS())) {b=true}
  n41 = Clamp(None=Multiply(B=Power(Base=@n40) {constExponent=0.3}) {constA=2}) {maxDefault=0.5}
  n42 = NamedRerouteUsage()
  n43 = NamedRerouteUsage()
  n44 = Multiply(A=@n43, B=Constant() {r=3}) {constB=3}
  n45 = Time()
  n46 = Multiply(B=ScalarParameter() {defaultValue=0.2, group=0_EmissiveColor, parameterName=Speed}) {constA=-1}
  n47 = TextureCoordinate() {coordinateIndex=1, uTiling=2}
  n48 = Sine(None=Add(A=Multiply(A=@n45, B=@n46), B=ComponentMask(None=@n47) {r=true}))
  n49 = Clamp(None=Multiply(A=Add(A=@n48)) {constB=0.5}) {minDefault=0.2}
  n50 = Multiply(A=Multiply(A=Add(A=@n41, B=@n42), B=LinearInterpolate(A=@n43, B=@n44, Alpha=@n49)), B=Constant() {r=2})
  NamedRerouteDeclaration(None=Multiply(A=Multiply(A=@n39, B=@n50), B=ScalarParameter() {defaultValue=0.2, group=0_EmissiveColor, parameterName=EmissivePower, sortPriority=4})) {name=FresnalColor, nodeColor=<rgba 0.1016 1 0 1>}
  n51 = TextureSampleParameter2D() {
   group=0_EmissiveColor
   parameterName=Rim_Normal
   samplerType=SAMPLERTYPE_Normal
   sortPriority=0
   texture=/VehicleCosmetics/Wheels/StuddedSE/Textures/T_Rim_StuddedSE_N
  }
  NamedRerouteDeclaration(None=@n51.RGB) {name=RimNormal, nodeColor=<rgba 1 0 0.8906 1>}
  n52 = TextureSample() {samplerType=SAMPLERTYPE_LinearColor, texture=/VehicleCosmetics/Wheels/StuddedSE/Textures/T_Tire_StuddedSE_RGB}
  NamedRerouteDeclaration(None=@n52.G) {name=TireGMask, nodeColor=<rgba 1 0 0.6094 1>}
  n53 = TextureSampleParameter2D() {
   group=0_EmissiveColor
   parameterName=Rim_Mask
   samplerType=SAMPLERTYPE_LinearColor
   sortPriority=1
   texture=/VehicleCosmetics/Wheels/StuddedSE/Textures/T_Rim_StuddedSE_RGB
  }
  NamedRerouteDeclaration(None=@n53.G) {name=RimGMask, nodeColor=<rgba 0 0.9297 1 1>}
  n54 = VectorParameter() {group=1_BaseColor, parameterName=AccentColor, sortPriority=0, defaultValue=<rgba 0.0078 0.0353 0.0235 1>}
  NamedRerouteDeclaration(None=@n54.RGB) {name=Accent_Color, nodeColor=<rgba 1 0.75 0 1>}
  NamedRerouteDeclaration(None=@n53.A) {name=RimAMask, nodeColor=<rgba 0 0.7187 1 1>}
  n55 = VectorParameter() {group=1_BaseColor, parameterName=PrimaryColor, sortPriority=5, defaultValue=<rgba 0.0232 0.0232 0.0232 1>}
  NamedRerouteDeclaration(None=@n55.RGB) {name=Primary_Color, nodeColor=<rgba 0.8984 0 1 1>}
  n56 = TextureSampleParameter2D() {
   group=1_BaseColor
   parameterName=Param
   samplerType=SAMPLERTYPE_LinearColor
   sortPriority=4
   texture=/VehicleCosmetics/Wheels/StuddedSE/Textures/T_Rim_StuddedSE_Mask
  }
  NamedRerouteDeclaration(None=@n56.R) {name=MaskR, nodeColor=<rgba 0 1 0.5781 1>}
  n57 = Time()
  n58 = ScalarParameter() {defaultValue=1, group=2_GradientColor, parameterName=GradientTextureSpeed, sortPriority=1}
  n59 = TextureSampleParameter2D(UVs=Panner(Coordinate=TextureCoordinate(), Time=Multiply(A=@n57, B=@n58)) {speedX=-0.1}) {
   group=2_GradientColor
   parameterName=GradientTexture
   sortPriority=0
   texture=/VehicleCosmetics/Wheels/StuddedSE/Textures/T_StuddedSE_Diffuse
  }
  NamedRerouteDeclaration(None=@n59.RGB) {name=GradientColor, nodeColor=<rgba 0 1 0.9766 1>}
  NamedRerouteDeclaration(None=@n53.R) {name=RimRMask, nodeColor=<rgba 0 1 0.7891 1>}
  n60 = VectorParameter() {group=0_EmissiveColor, parameterName=EmissiveColor, sortPriority=2, defaultValue=<rgba 0.0431 0.5333 0.1529 1>}
  NamedRerouteDeclaration(None=@n60.RGB) {name=CustomColor, nodeColor=<rgba 0 0.8125 1 1>}
  NamedRerouteDeclaration(None=@n53.B) {name=RimBMask, nodeColor=<rgba 0 1 0.1094 1>}
  NamedRerouteDeclaration(None=TextureSample() {samplerType=SAMPLERTYPE_Normal, texture=/VehicleCosmetics/Wheels/StuddedSE/Textures/T_Tire_StuddedSE_N}.RGB) {name=TireNormal, nodeColor=<rgba 0 1 0.1328 1>}
  NamedRerouteDeclaration(None=@n52.R) {name=TireRMask, nodeColor=<rgba 0.125 0 1 1>}
  NamedRerouteDeclaration(None=VertexColor().R) {name=TireRMaskVC, nodeColor=<rgba 0.1953 1 0 1>}
  NamedRerouteDeclaration(None=@n52.B) {name=TireBMask, nodeColor=<rgba 0.1016 0 1 1>}
  NamedRerouteDeclaration(None=Constant4Vector() {constant=<rgba 0 0 0 0.02>}.RGB) {name=Secondary_Color, nodeColor=<rgba 0.3594 0 1 1>}
 ]
}
