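# Wheel material for the StuddedSE vehicle cosmetic skeletal mesh, with animated rim and tire color effects.
# It overrides base color, emissive and opacity mask on MF_WheelParent_Simple's attributes, blending colors through rim, tire and alpha masks.
# It declares a time-pulsed Fresnel glow, switchable between LOD0 and LOD2 detail masks, and a panning gradient texture.
# The result passes through the vehicle damage function, then the de-rez transition.
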Material /VehicleCosmetics/Wheels/StuddedSE/Materials/M_Wheel_StuddedSE_Fix
graph {
 properties {
  bAutomaticallySetUsageInEditor=false
  bUseMaterialAttributes=true
  bUsedWithMeshDeformer=true
  bUsedWithSkeletalMesh=true
  pixelDepthOffsetMode=PDOM_Legacy
  previewMesh=/VehicleCosmetics/Wheels/StuddedSE/Meshes/SK_Wheel_StuddedSE
  usedShadingModels=Default Lit
 }
 nodes [
  n1 = NamedRerouteUsage()
  n2 = NamedRerouteUsage()
  n3 = ComponentMask(None=NamedRerouteUsage()) {r=true}
  n4 = LinearInterpolate(
   A=LinearInterpolate(A=NamedRerouteUsage(), B=NamedRerouteUsage(), Alpha=ComponentMask(None=@n2) {b=true}) {desc=Rim}
   B=NamedRerouteUsage()
   Alpha=@n3
  )
  n5 = NamedRerouteUsage()
  n6 = LinearInterpolate(
   A=@n4
   B=LinearInterpolate(A=NamedRerouteUsage(), B=NamedRerouteUsage(), Alpha=ComponentMask(None=@n5) {r=true}) {desc=Tire}
   Alpha=NamedRerouteUsage()
  )
  n7 = Multiply(A=Multiply(A=@n4, B=NamedRerouteUsage()), B=OneMinus(None=NamedRerouteUsage()))
  n8 = TextureSample() {
   samplerSource=SSM_Wrap_WorldGroupSettings
   samplerType=SAMPLERTYPE_Masks
   texture=/VehicleCosmetics/Wheels/StuddedSE/Textures/T_Rim_StuddedSE_AlphaMask
  }
  n9 = LinearInterpolate(A=@n8.R, Alpha=NamedRerouteUsage())
  MaterialFunctionCall(In=MaterialFunctionCall(Material=SetMaterialAttributes!(MaterialAttributes=@n1, Base Color=@n6, Emissive Color=@n7, Opacity Mask=@n9) {attributeSetTypes [#1, #2, #3]}) {materialFunction=&MF_BaseVehicle_Damage}.Result) {materialFunction=&MF_VehicleDeRez_Transition}
  NamedRerouteDeclaration(None=@n8.G) {name=Detail_Out_LOD0, nodeColor=<rgba 1 0 0.5391 1>}
  NamedRerouteDeclaration(None=@n8.B) {name=Detail_Out_LOD2, nodeColor=<rgba 0 0.8594 1 1>}
  n10 = VectorParameter() {group=5_Special Effects, parameterName=EmissiveColor, sortPriority=2, defaultValue=<rgba 0.0431 0.5333 0.1529 1>}
  NamedRerouteDeclaration(None=@n10.RGB) {name=EmissiveColor, nodeColor=<rgba 0 0.8125 1 1>}
  n11 = NamedRerouteUsage()
  n12 = Saturate(None=Fresnel(Normal=Transform(None=@n11)) {baseReflectFraction=0.4, exponent=2})
  n13 = Multiply(A=LinearInterpolate(
    A=NamedRerouteUsage()
    B=NamedRerouteUsage()
    Alpha=ScalarParameter() {desc="0== LOD0; 1==LOD2", group=5_Special Effects, parameterName=DetailIsLOD2}
   ) {constAlpha=0}) {constB=0.5}
  n14 = NamedRerouteUsage()
  n15 = Multiply(A=@n14, B=Constant() {r=5}) {constB=3}
  n16 = Time()
  n17 = Multiply(B=Constant() {desc=Fresnel Speed, r=0.2}) {constA=-1}
  n18 = TextureCoordinate() {coordinateIndex=1, uTiling=4}
  n19 = Sine(None=Add(A=Multiply(A=@n16, B=@n17), B=ComponentMask(None=@n18) {r=true}))
  n20 = Clamp(None=Multiply(A=Add(A=@n19)) {constB=0.5, desc=Dim brightness}) {minDefault=0.2}
  n21 = Multiply(A=Multiply(A=Clamp(None=@n13) {maxDefault=0.5, minDefault=0.1}, B=LinearInterpolate(A=@n14, B=@n15, Alpha=@n20)), B=Constant() {r=2})
  NamedRerouteDeclaration(None=Multiply(A=Multiply(A=@n12, B=@n21), B=ScalarParameter() {defaultValue=0.2, group=5_Special Effects, parameterName=EmissivePower, sortPriority=4})) {name=FresnelColor, nodeColor=<rgba 0.1016 1 0 1>}
  n22 = TextureSampleParameter2D(UVs=Panner() {speedX=-0.2}) {
   desc=Diffuse Input Texture - Optional
   group=1_Input Textures
   parameterName=Rim_DiffuseTexture
   samplerSource=SSM_Wrap_WorldGroupSettings
   texture=/VehicleCosmetics/Textures/T_White
  }
  NamedRerouteDeclaration(None=@n22.RGB) {name=GradientColor, nodeColor=<rgba 0 1 0.9766 1>}
  n23 = MaterialFunctionCall() {materialFunction=&MF_WheelParent_Simple}
  n24 = GetMaterialAttributes(None=@n23.MaterialAttributes) {attributeGetTypes [#1, #4, #5]}
  NamedRerouteDeclaration(None=@n24.Normal) {name=MA_Normal_Out, nodeColor=<rgba 0.9453 1 0 1>}
  NamedRerouteDeclaration(None=@n23.MaterialAttributes) {name=MA_Out, nodeColor=<rgba 0.9688 0 1 1>}
  NamedRerouteDeclaration(None=@n23.RimRGBAMask) {name=Rim_RGBA, nodeColor=<rgba 0 1 0.6484 1>}
  NamedRerouteDeclaration(None=@n23.TireRGBAMask) {name=Tire_RGBA, nodeColor=<rgba 1 0.3281 0 1>}
  NamedRerouteDeclaration(None=@n23.VertexColorR) {name=VertexColor_Out, nodeColor=<rgba 0.0078 0 1 1>}
  NamedRerouteDeclaration(None=@n24.Base Color) {name=MA_BaseColor_Out, nodeColor=<rgba 0.6406 0 1 1>}
  NamedRerouteDeclaration(None=@n3) {name=MA_RimR_Out, nodeColor=<rgba 0 1 0.9531 1>}
 ]
}
